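# Conversation database for the Frost Galore NPC, providing its greeting and the branches offered after it.
# The entry point leads to a speech node with a default greeting and prioritized contextual lines for first-ever meetings, match conversation count and two specific player outfits.
# Link nodes then jump to shared NPC quest, basic services, Season 38 story and LAF POI conversations.
# The LAF link requires that the player lacks the active LAF quest tag.
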ConversationDatabase /NPCLibrary/NPCs/FrostGalore/NPCLibrary_Conversation_FrostGalore
entry_points [{tag=Conversation.Entry.FrostGalore, destinations [#1]}]
nodes [
 ConversationEntryPointNode_7 {
  id=#1
  class=ConversationEntryPointNode
  outputs [#2]
  properties {entryTag {tagName=Conversation.Entry.FrostGalore}}
 }
 Speech_C_4 {
  id=#2
  class=Speech_C
  outputs [
   #3
   #4
   #5
   #6
   #7
  ]
  properties {
   nodeName=""
   generalConfig {
    defaultMessage="Hi!"
    contextualMessages [
     {
      message=Sorry about my dad... and my brother... and Springfield.
      contextRequirements [{requirement {class=FortControllerRequirement_IsFirstEverConversationWithNPC}}]
      priority {curve {rowName=Tandem.Speech.Priority.FirstEverInteraction}}
     }
     {
      message="Hi!"
      contextRequirements [{requirement {class=FortControllerRequirement_CheckNPCMatchConversationCount, properties {nPCUniqueIdTag {tagName=AISpawnerData.Type.Tandem.FrostGalore}}}}]
      priority {curve {rowName=None}}
     }
     {
      message=Do you think other kids have to save the world from their dad as often as we do?
      contextRequirements [{requirement {class=FortControllerRequirement_HasCID, properties {allowedCIds [/BRCosmetics/Athena/Items/Cosmetics/Characters/Character_QueenTruth_Rind]}}}]
      priority {value=2, curve {rowName=Tandem.Speech.Priority.OutfitReaction}}
     }
     {
      message="Dad, you know I love you, but please don't embarrass us in front of the other realities."
      contextRequirements [{requirement {class=FortControllerRequirement_HasCID, properties {allowedCIds [/BRCosmetics/Athena/Items/Cosmetics/Characters/Character_SilentNovel_Vane]}}}]
      priority {value=2, curve {rowName=Tandem.Speech.Priority.OutfitReaction}}
     }
    ]
   }
  }
 }
 ConversationLinkNode_4 {
  id=#7
  class=ConversationLinkNode
  properties {remoteEntryTag {tagName=Conversation.Entry.Season38.LAF.NPC.POI}}
  sub_nodes [FortConversationRequirement_ControllerRequirement_1 {
    class=FortConversationRequirement_ControllerRequirement
    properties {
     nodeName="Require: Has not started a L&&F Quest"
     requirement {class=FortControllerRequirement_HasOwnedTag, properties {bInverseCondition=true, ownedTag {tagName=Athena.Quests.S38.LAF.QuestActive}}}
    }
   }]
 }
 ConversationLinkNode_3 {
  id=#6
  class=ConversationLinkNode
  properties {remoteEntryTag {tagName=Conversation.Entry.Season38.Story.P03.Q07.02}}
 }
 ConversationLinkNode_2 {
  id=#5
  class=ConversationLinkNode
  properties {remoteEntryTag {tagName=Conversation.Entry.Season38.Story.P03.Q03}}
 }
 ConversationLinkNode_0 {id=#4, class=ConversationLinkNode, properties {remoteEntryTag {tagName=Conversation.Entry.BasicServices}}}
 ConversationLinkNode_1 {id=#3, class=ConversationLinkNode, properties {remoteEntryTag {tagName=Conversation.Entry.NPCLibrary.Quests}}}
]
